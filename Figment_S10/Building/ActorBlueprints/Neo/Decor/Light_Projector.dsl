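; Light_Projector is a decor actor Blueprint in Figment_S10 that shows a configurable holographic pattern on its S_HoloWall_01 component.
; After calling the parent, its construction script creates a dynamic instance of M_HoloDecor on that component's first material slot.
; It feeds Projector and Mask variables into material parameters for pattern, motion, colors and mask texture.
; Transition is lerped into a narrow range, and mask and wave scales are multiplied into larger shader values.
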(fn ConstructionScript
  (|CallParentFunction)
  (bind _returnvalue (Rendering|Material|CreateDynamicMaterialInstance (Variables|Default|GetS_HoloWall_01) 0 "/Figment_S10/Building/ActorBlueprints/Neo/Decor/M_HoloDecor.M_HoloDecor" "None" :!class /Script/Engine.PrimitiveComponent))
  (Rendering|Material|SetScalarParameterValue _returnvalue "PatternTransparency" (Variables|Projector|GetShowPattern) :!class /Script/Engine.MaterialInstanceDynamic)
  (Rendering|Material|SetScalarParameterValue _returnvalue "PatternTransition" (Math|Float|Lerp 0.2 0.7 (Variables|Projector|GetPatternTransition)) :!class /Script/Engine.MaterialInstanceDynamic)
  (Rendering|Material|SetScalarParameterValue _returnvalue "Speed" (Variables|Projector|GetSpeed/Direction) :!class /Script/Engine.MaterialInstanceDynamic)
  (Rendering|Material|SetScalarParameterValue _returnvalue "R Scale" (* (Variables|Mask|GetR%20Scale) 512.0) :!class /Script/Engine.MaterialInstanceDynamic)
  (Rendering|Material|SetScalarParameterValue _returnvalue "G Scale" (* (Variables|Mask|GetG%20Scale) 512.0) :!class /Script/Engine.MaterialInstanceDynamic)
  (Rendering|Material|SetScalarParameterValue _returnvalue "B Scale" (* (Variables|Mask|GetB%20Scale) 512.0) :!class /Script/Engine.MaterialInstanceDynamic)
  (Rendering|Material|SetScalarParameterValue _returnvalue "Noise Scale" (* (Variables|Mask|GetWave%20Scale) 1.8e+03) :!class /Script/Engine.MaterialInstanceDynamic)
  (Rendering|Material|SetScalarParameterValue _returnvalue "Wave" (Variables|Projector|GetWave) :!class /Script/Engine.MaterialInstanceDynamic)
  (Rendering|Material|SetScalarParameterValue _returnvalue "Rotation" (Variables|Projector|GetRotation) :!class /Script/Engine.MaterialInstanceDynamic)
  (Rendering|Material|SetTextureParameterValue _returnvalue "Mask" (Variables|Mask|GetMask))
  (Rendering|Material|SetVectorParameterValue _returnvalue "Color A" (Variables|Projector|GetColor%20A) :!class /Script/Engine.MaterialInstanceDynamic)
  (Rendering|Material|SetVectorParameterValue _returnvalue "Color B" (Variables|Projector|GetColor%20B) :!class /Script/Engine.MaterialInstanceDynamic)
  (Rendering|Material|SetVectorParameterValue _returnvalue "Color Back" (Variables|Projector|GetColor%20Background) :!class /Script/Engine.MaterialInstanceDynamic))
