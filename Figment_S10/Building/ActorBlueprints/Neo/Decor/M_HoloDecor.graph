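# Additive unlit material for holographic Neo decor, showing animated colored patterns over a height-faded back color.
# The mask's RGB channels are world-projected on two planes blended by normal X, then panned with noise-modulated speed and an actor-position time offset.
# They drive HeightLerp between back and pattern colors, faded by mask alpha, UV and noise.
# Color_Time_of_Day scales the result with day and night multipliers.
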Material /Figment_S10/Building/ActorBlueprints/Neo/Decor/M_HoloDecor
graph {
 properties {blendMode=BLEND_Additive, pixelDepthOffsetMode=PDOM_Legacy, shadingModel=MSM_Unlit, usedShadingModels=Unlit}
 nodes [
  n1 = ActorPositionWS()
  n2 = Add(
   A=Divide(
    A=MaterialFunctionCall!(f2, f3=@n1, f4) {materialFunction=/Game/Packages/Fortnite_UniqueMaterials_1/VectorComponentOps}.f3 Add
    B=Constant() {r=256}
   )
   B=Time()
  )
  Frac(None=@n2)
  n3 = ScalarParameter() {defaultValue=0.2, group=Emissive, parameterName=DayMult, sortPriority=1}
  n4 = WorldPosition()
  n5 = ActorPositionWS()
  n6 = Divide(
   A=Transform(None=Subtract(A=@n4.XYZ, B=@n5)) {transformSourceType=TRANSFORMSOURCE_World, transformType=TRANSFORM_Local}
   B=Constant() {r=100}
  )
  n7 = Saturate(None=OneMinus(None=ComponentMask(None=@n6) {b=true}))
  n8 = VectorParameter() {parameterName=Color Back, defaultValue=<rgba 0 0.0312 0.0677 0>}
  n9 = VectorParameter() {parameterName=Color A, defaultValue=<rgba 0.2549 0.3412 0.0784 0>}
  n10 = VectorParameter() {parameterName=Color B, defaultValue=<rgba 0.6354 0.2149 0.0284 0>}
  n11 = WorldPosition()
  n12 = Multiply(A=Divide(A=@n11.XYZ, B=ScalarParameter() {defaultValue=512, parameterName=G Scale})) {constB=-1}
  n13 = ComponentMask(None=@n12) {b=true, r=true}
  n14 = ScalarParameter() {defaultValue=0.5, parameterName=Wave}
  n15 = Add(A=@n14) {constB=0.5}
  n16 = Multiply(A=Divide(A=WorldPosition().XYZ, B=ScalarParameter() {defaultValue=1800, parameterName=Noise Scale})) {constB=-1}
  n17 = Multiply(
   A=VectorParameter() {parameterName=Noise Speed, defaultValue=<rgba 0.05 -0.1 0 0>}.RGB
   B=ScalarParameter() {defaultValue=-1, parameterName=Speed}
  )
  n18 = TextureSample(UVs=Panner(Coordinate=ComponentMask(None=@n16) {b=true, r=true}, Speed=@n17)) {samplerType=SAMPLERTYPE_Grayscale, texture=/Figment_S10/VisualThreatManager/StormVisuals/Test/SpawnParticles/Streamers/LowResBlurredNoise}
  n19 = TextureSample(UVs=Panner(Coordinate=ComponentMask(None=@n16) {g=true, r=true}, Speed=@n17)) {samplerType=SAMPLERTYPE_Grayscale, texture=&LowResBlurredNoise}
  n20 = VertexNormalWS()
  n21 = LinearInterpolate(A=@n18.R, B=@n19.R, Alpha=Abs(None=ComponentMask(None=@n20) {r=true}))
  n22 = ComponentMask(None=TextureCoordinate()) {g=true}
  n23 = Constant() {r=0.1}
  n24 = Multiply(A=Constant() {r=0.0001}, B=Saturate(None=Subtract(A=@n22, B=@n23)))
  n25 = Add(A=Multiply(A=Saturate(None=@n15), B=Multiply(A=@n21, B=@n24)), B=ScalarParameter() {defaultValue=-1, parameterName=Speed})
  n26 = Multiply(A=@n25, B=VectorParameter() {parameterName=G Speed, defaultValue=<rgba -0.05 0.05 0 0>}.RGB)
  n27 = ScalarParameter() {parameterName=Rotation}
  n28 = Multiply(A=Time(), B=Multiply(A=@n27, B=Constant() {r=0}))
  n29 = TextureSampleParameter2D(UVs=Rotator(Coordinate=Panner(Coordinate=@n13, Time=@n2, Speed=@n26), Time=@n28)) {group=Mask, parameterName=Mask, samplerType=SAMPLERTYPE_Masks, texture=/Figment_S10/Environments/Sets/Neo/Props/Textures/T_Decor_M}
  n30 = ComponentMask(None=@n12) {b=true, g=true}
  n31 = TextureSampleParameter2D(UVs=Rotator(Coordinate=Panner(Coordinate=@n30, Time=@n2, Speed=@n26), Time=@n28)) {group=Mask, parameterName=Mask, samplerType=SAMPLERTYPE_Masks, texture=&T_Decor_M}
  n32 = VertexNormalWS()
  n33 = LinearInterpolate(A=@n29.G, B=@n31.G, Alpha=Abs(None=ComponentMask(None=@n32) {r=true}))
  n34 = Multiply(A=Divide(A=@n11.XYZ, B=ScalarParameter() {defaultValue=512, parameterName=R Scale})) {constB=-1}
  n35 = ComponentMask(None=@n34) {b=true, r=true}
  n36 = Multiply(A=VectorParameter() {parameterName=R Speed, defaultValue=<rgba 0.05 0.05 0 0>}.RGB, B=@n25)
  n37 = Multiply(A=Time(), B=Multiply(A=Constant() {r=0.0001}, B=@n27))
  n38 = TextureSampleParameter2D(UVs=Rotator(Coordinate=Panner(Coordinate=@n35, Time=@n2, Speed=@n36), Time=@n37)) {group=Mask, parameterName=Mask, samplerType=SAMPLERTYPE_Masks, texture=&T_Decor_M}
  n39 = ComponentMask(None=@n34) {g=true, r=true}
  n40 = TextureSampleParameter2D(UVs=Rotator(Coordinate=Panner(Coordinate=@n39, Time=@n2, Speed=@n36), Time=@n37)) {group=Mask, parameterName=Mask, samplerType=SAMPLERTYPE_Masks, texture=&T_Decor_M}
  n41 = ComponentMask(None=VertexNormalWS()) {r=true}
  n42 = Min(A=LinearInterpolate(A=@n38.R, B=@n40.R, Alpha=Abs(None=@n41)), B=@n33)
  n43 = Multiply(A=Divide(A=@n11.XYZ, B=ScalarParameter() {defaultValue=512, parameterName=B Scale})) {constB=-1}
  n44 = ComponentMask(None=@n43) {b=true, r=true}
  n45 = Multiply(A=@n25, B=VectorParameter() {parameterName=B Speed, defaultValue=<rgba -0.05 0.1 0 0>}.RGB)
  n46 = Multiply(A=Time(), B=Multiply(A=@n27, B=Constant() {r=0.0002}))
  n47 = TextureSampleParameter2D(UVs=Rotator(Coordinate=Panner(Coordinate=@n44, Time=@n2, Speed=@n45), Time=@n46)) {group=Mask, parameterName=Mask, samplerType=SAMPLERTYPE_Masks, texture=&T_Decor_M}
  n48 = ComponentMask(None=@n43) {b=true, g=true}
  n49 = TextureSampleParameter2D(UVs=Rotator(Coordinate=Panner(Coordinate=@n48, Time=@n2, Speed=@n45), Time=@n46)) {group=Mask, parameterName=Mask, samplerType=SAMPLERTYPE_Masks, texture=&T_Decor_M}
  n50 = VertexNormalWS()
  n51 = LinearInterpolate(A=@n47.B, B=@n49.B, Alpha=Abs(None=ComponentMask(None=@n50) {r=true}))
  n52 = TextureSampleParameter2D() {group=Mask, parameterName=Mask, samplerType=SAMPLERTYPE_Masks, texture=&T_Decor_M}
  n53 = ComponentMask(None=TextureCoordinate()) {g=true}
  n54 = Multiply(
   A=MaterialFunctionCall(
    A=@n8.RGB
    B=LinearInterpolate(A=@n9.RGB, B=@n10.RGB, Alpha=@n33)
    Transition Phase=ScalarParameter() {defaultValue=0.75, parameterName=PatternTransition}
    Height Texture=Max(A=@n42, B=@n51)
    Contrast=Constant() {r=0.5}
   ) {materialFunction=&HeightLerp}.Results
   B=Subtract(A=Multiply(A=@n52.A, B=@n53), B=Multiply(A=@n21, B=@n14))
  )
  n55 = ScalarParameter() {defaultValue=1, parameterName=PatternTransparency}
  Multiply(
   A=MaterialFunctionCall(
    Color_Morning=@n3
    Color_Day=@n3
    Color_Evening=@n3
    Color_Night=ScalarParameter() {defaultValue=1, group=Emissive, parameterName=NightMult, sortPriority=2}
   ) {materialFunction=&Color_Time_of_Day}.Result
   B=Add(A=Multiply(A=@n7, B=@n8.RGB), B=Multiply(A=@n54, B=@n55))
  )
 ]
}
